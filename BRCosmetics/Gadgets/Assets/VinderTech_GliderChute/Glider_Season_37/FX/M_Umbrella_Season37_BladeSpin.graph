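# Translucent, unlit, two-sided Niagara mesh-particle material for the spinning-blade effect on the Season 37 umbrella glider.
# Emissive fakes lighting with time-of-day tint, sun/moon reflection highlights and a height gradient around the particle, using a cheaper color path on mobile.
# Two panning noise textures blend into streaks modulating emissive and opacity.
# Opacity also combines V-gradient, particle alpha, vertex red, depth fade and a switchable view-angle or U-sine fade.
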Material /BRCosmetics/Gadgets/Assets/VinderTech_GliderChute/Glider_Season_37/FX/M_Umbrella_Season37_BladeSpin
graph {
 properties {
  bUsedWithNiagaraMeshParticles=true
  blendMode=BLEND_Translucent
  pixelDepthOffsetMode=PDOM_Legacy
  shadingModel=MSM_Unlit
  twoSided=true
  usedShadingModels=Unlit
 }
 nodes [
  n1 = Multiply(A=Add(A=MaterialFunctionCall() {materialFunction=&Color_Time_of_Day}.Result) {constB=0.25}) {constB=0.5}
  n2 = CollectionParameter() {
   collection=&FortniteMaterialParameters
   parameterName=SunAndMoonModelDirectionalVector
  }
  n3 = DotProduct(A=ComponentMask(None=@n2) {b=true, g=true, r=true}, B=ReflectionVectorWS() {bNormalizeCustomWorldNormal=true})
  n4 = ScalarParameter() {defaultValue=16, parameterName=Spec Highlight Power}
  n5 = Multiply(A=@n1, B=Clamp(None=Power(Base=@n3, Exp=@n4) {constExponent=20}))
  n6 = Multiply(A=@n2) {constB=-1}
  n7 = Power(
   Base=DotProduct(A=ComponentMask(None=@n6) {b=true, g=true, r=true}, B=ReflectionVectorWS() {bNormalizeCustomWorldNormal=true})
   Exp=ScalarParameter() {defaultValue=25, parameterName=Reverse Spec Highlight Power}
  ) {constExponent=20}
  n8 = Multiply(A=Multiply(A=@n1, B=Clamp(None=@n7)), B=ScalarParameter() {defaultValue=1, parameterName=ReverseHighlightOpacity})
  n9 = LinearInterpolate(B=MaterialFunctionCall() {materialFunction=&Color_Time_of_Day}.Result) {constA=1, constAlpha=0.6}
  n10 = VectorParameter() {parameterName=TireColor, defaultValue=<rgba 0.1615 0.1615 0.1615 0>}
  n11 = ParticleColor()
  n12 = Add(A=Multiply(A=VectorParameter() {parameterName=BaseColorMult, defaultValue=<rgba 1 1 1 0>}.RGB, B=@n11.RGB)) {constB=0.15}
  n13 = VectorParameter() {parameterName=FakeReflectColor, defaultValue=<rgba 0.1615 0.1497 0.1497 1>}
  n14 = ParticlePositionWS()
  n15 = Constant3Vector() {constant=<rgba 0 0 28 0>}
  n16 = ComponentMask(None=Subtract(A=Add(A=@n14, B=@n15), B=WorldPosition().XYZ)) {b=true}
  n17 = Clamp(None=Power(Base=Divide(A=@n16) {constB=60}) {constExponent=2.2})
  n18 = VertexColor()
  n19 = Multiply(A=StaticSwitchParameter(True=Multiply(A=@n13.RGB, B=@n17), False=@n17) {parameterName=UseFakeSpecColor}, B=@n18.R)
  n20 = Add(
   A=StaticSwitchParameter(True=Add(A=@n5, B=@n8), False=@n5) {parameterName=AddReversedFakeLightVector}
   B=LinearInterpolate(A=Multiply(A=@n9, B=@n10.RGB), B=@n12, Alpha=Clamp(None=@n19))
  )
  n21 = MaterialFunctionCall(Float2=TextureCoordinate()) {materialFunction=&BreakOutFloat2Components}
  n22 = AppendVector(
   A=Multiply(A=@n21.R, B=ScalarParameter() {defaultValue=1, parameterName=Utile_A})
   B=Multiply(A=@n21.G, B=ScalarParameter() {defaultValue=1, parameterName=Vtile_A}) {constB=1.25}
  )
  n23 = Multiply(A=Time(), B=ScalarParameter() {parameterName=PanSpeed})
  n24 = MaterialFunctionCall(Float2=TextureCoordinate()) {materialFunction=&BreakOutFloat2Components}
  n25 = ScalarParameter() {defaultValue=1, parameterName=Utile_B}
  n26 = ScalarParameter() {defaultValue=1, parameterName=Vtile_B}
  n27 = Add(A=AppendVector(A=Multiply(A=@n24.R, B=@n25) {constB=0.6}, B=Multiply(A=@n24.G, B=@n26) {constB=2})) {constB=0.2}
  n28 = Multiply(A=@n23) {constB=1.2}
  n29 = LinearInterpolate(
   A=TextureSample(UVs=Panner(Coordinate=@n22, Time=@n23) {speedX=-3, speedY=-0.01}) {samplerType=SAMPLERTYPE_Grayscale, texture=&LowResBlurredNoise}.G
   B=TextureSample(UVs=Panner(Coordinate=@n27, Time=@n28) {speedX=-3, speedY=0.05}) {samplerType=SAMPLERTYPE_Grayscale, texture=&LowResBlurredNoise}.G
  )
  MaterialFunctionCall(Emissive=Multiply(A=ShadingPathSwitch(Default=@n20, Mobile=@n12) {desc=Migrated from FeatureLevelSwitch}, B=@n29)) {materialFunction=&MF_FX_EmissiveAdjustments}
  n30 = MaterialFunctionCall() {materialFunction=&LinearGradient}
  n31 = Multiply(A=@n11.A, B=@n29)
  n32 = VertexColor()
  n33 = QualitySwitch(
   Default=DepthFade(
    Opacity=ScalarParameter() {defaultValue=24, parameterName=BaseOpacity}
    FadeDistance=ScalarParameter() {defaultValue=10, parameterName=FadeDistance}
   ) {fadeDistanceDefault=3}
   Low=Constant() {r=1}
  )
  n34 = MaterialFunctionCall(Float2=TextureCoordinate()) {materialFunction=&BreakOutFloat2Components}
  n35 = Saturate(None=Sine(None=Multiply(A=@n34.R) {constB=0.5}))
  n36 = MaterialFunctionCall("FadeSensitivity (optional)"=ScalarParameter() {defaultValue=1, parameterName=AngleFade}) {materialFunction=&FadeBasedOnViewAngle}
  n37 = StaticSwitchParameter(
   True=Multiply(A=LinearInterpolate(Alpha=@n32.R) {constA=0.5}, B=Multiply(A=@n33, B=@n35))
   False=Multiply(A=@n18.R, B=Multiply(A=@n33, B=@n36.Result))
  ) {parameterName=UseMasktoFade}
  Multiply(
   A=Power(Base=OneMinus(None=@n30.VGradient), Exp=ScalarParameter() {defaultValue=2, parameterName=Mask_Power}) {constExponent=5}
   B=Multiply(A=Multiply(A=@n31, B=@n37), B=ScalarParameter() {defaultValue=1, parameterName=Opacity})
  )
  ObjectPositionWS()
 ]
}
